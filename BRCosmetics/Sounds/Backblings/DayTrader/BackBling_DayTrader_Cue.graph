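# Sound cue for the DayTrader back bling cosmetic, playing a single wave with pitch driven by a KillCount parameter.
# A continuous modulator maps kill count to a bounded pitch range at the root, fed by a short fixed delay before the wave player.
# It uses the close-range pet attenuation, the Backblings sound class and the reactive back bling audio property sheet.
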SoundCue /BRCosmetics/Sounds/Backblings/DayTrader/BackBling_DayTrader_Cue
root=ModulatorContinuous_0
nodes [
 {id=Delay_0, children [WavePlayer_1], properties {delayMax=0.1, delayMin=0.1}}
 {
  id=ModulatorContinuous_0
  children [Delay_0]
  properties {pitchModulationParams {default=0, maxInput=15, maxOutput=1.2, minOutput=0.8, parameterName=KillCount}}
 }
 {id=WavePlayer_1, properties {soundWaveAssetPtr=&/BackBling_DayTrader_01}}
]
properties {
 attenuationSettings=&Fort_Pets_Close_Athena_Att
 audioPropertiesSheet=/BRCosmetics/Sounds/Settings/PropertySheets/Reactive/APS_Cosmetic_Reactive_Backbling
 maxDistance=1800.0001
 soundClassObject=/Game/Sounds/FortSoundClassesAndModes/Fort_Backblings
 volumeMultiplier=1.2
}
